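; Athena playlist mutator, filed under ItemTest, that overrides player starting and maximum health and shield.
; On BeginPlay it calls the parent, then writes values resolved from the HealthValue and ShieldValue scalable floats into its Health properties.
; Each scalable float drives both the starting and max setting, so health and shield both start full.
; The construction script is empty.
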(event EventBeginPlay
  (bind _healthvalue (Variables|Default|GetHealthValue))
  (bind _returnvalue (ScalableFloat|GetValueatLevel _healthvalue))
  (bind _shieldvalue (Variables|Default|GetShieldValue))
  (bind _returnvalue_2 (ScalableFloat|GetValueatLevel _shieldvalue))
  (|CallParentFunction)
  (Variables|Health|SetStartingHealth _returnvalue)
  (Variables|Health|SetMaxHealth _returnvalue)
  (Variables|Health|SetStartingShield _returnvalue_2)
  (Variables|Health|SetMaxShield _returnvalue_2))

(fn ConstructionScript)
